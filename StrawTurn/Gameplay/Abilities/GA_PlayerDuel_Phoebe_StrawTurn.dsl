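; Gameplay ability that resolves the end of a StrawTurn Phoebe duel quest for the duel's instigator.
; It overrides the parent's urgent-quest-ended hook, checks the owner's instigator tag, caches its ability system and switches on failed, poached or success states.
; Each outcome applies a gameplay effect.
; Success also fires a gameplay cue and spawns loot at the pawn, with the loot tier mapped from the playlist's squad-size tag.
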(event Ability|EventActivateAbility
  (|CallParentFunction) ; This node is disabled and will not be called.\nDrag off pins to build functionality.
)

(event Ability|EventOnEndAbility
  (params (bWasCancelled :type Boolean))
  (|CallParentFunction bWasCancelled) ; This node is disabled and will not be called.\nDrag off pins to build functionality.
)

(event EventAdditionalUrgentQuestEndedLogic
  (params (EnumDuelQuestState :type EnumDuelStateTrackingPhoebe%20Enum :default "NewEnumerator0"))
  ;[0 0 0] Get valid ability owner and their ability system, then determine what quest state their target entered
  (|CallParentFunction EnumDuelQuestState)
  (bind _bownerisinstigator (CallFunction|AbilityOwnerIsDuelInstigator))
  (if _bownerisinstigator
    (bind _abilitysystem (CallFunction|GetValidAbilitySystemforOwner))
    (Variables|Default|SetInstigatingPhoebeAbilitySystem _abilitysystem)
    (switch Utilities|FlowControl|Switch|SwitchonEnumDuelStateTrackingPhoebe EnumDuelQuestState
      (:NewEnumerator0
        (CallFunction|QuestFailed))
      (:NewEnumerator4
        (CallFunction|QuestPoached))
      (:NewEnumerator5
        (CallFunction|QuestSuccess)))))

(event Custom|QuestSuccess
  (bind _returnvalue (CallFunction|GetAvatarasFortPlayerPawn))
  ;[0 0 0] Call success logic on phoebe for the duel
  (GameplayEffects|ApplyGameplayEffectToSelf :self (Variables|Default|GetInstigatingPhoebeAbilitySystem) :GameplayEffectClass (Variables|Default|GetSuccessEffectClass) :Level 0.0)
  (Ability|ExecuteGameplayCueOnOwner (Variables|Default|GetSuccessGameplayCueTag))
  (Utilities|IsValid _returnvalue
    (:Is%20Valid
      (bind (_lootgroupname _bfoundtier) (CallFunction|GetLootTierforCurrentPlaylist))
      (if _bfoundtier
        (Class|AthenaFunctionLibrary|SpawnLootfromTable _lootgroupname 1 -1 (Transformation|GetActorLocation _returnvalue) "0, 0, 0" 0 true true)))))

(event Custom|QuestPoached
  ;[0 0 0] Logic for when the quest is failed or poached to execute on Phoebe
  (GameplayEffects|ApplyGameplayEffectToSelf :self (Variables|Default|GetInstigatingPhoebeAbilitySystem) :GameplayEffectClass "/StrawTurn/GE_StrawTurn_DuelPoachedOrEnded.GE_StrawTurn_DuelPoachedOrEnded_C" :Level 0.0))

(event Custom|QuestFailed
  ;[0 0 0] Logic for when the quest is failed or poached to execute on Phoebe
  (GameplayEffects|ApplyGameplayEffectToSelf :self (Variables|Default|GetInstigatingPhoebeAbilitySystem) :GameplayEffectClass (Variables|Default|GetFailedGameplayEffectClass) :Level 0.0))

(fn AbilityOwnerIsDuelInstigator
  (bind _returnvalue (CallFunction|GetAvatarasFortPlayerPawn))
  ;[0 0 0] Helper function for checking if the ability owner is also the duel instigator
  (Utilities|IsValid _returnvalue
    (:Is%20Valid
      (return (GameplayTags|HasMatchingGameplayTag _returnvalue "(TagName=\"Granted.Athena.JobBoardExperimental.Instigator\")")))))

(fn GetLootTierForCurrentPlaylist
  (bind (_value _returnvalue) (Utilities|Map|Find (Variables|Default|GetMapPlaylistToLootTier) (CallFunction|GetPlaylistSquadSizeTag)))
  ;[0 0 0] Maps the loot tier based on the playlist squad size and returns the name of the loot tier group
  (return _value _returnvalue))

(fn GetPlaylistSquadSizeTag
  (bind _returnvalue (Fortnite|Athena|GameState|GetGameStateAthena))
  ;[0 0 0] Get the playlist tag to determine if it's solo/squad etc.
  (Utilities|IsValid _returnvalue
    (:Is%20Valid
      (for _array_element (LocalVariables|GetPossiblePlaylistTags)
        (if (GameplayTags|TagContainer|CompareTagContainertoOtherTags _array_element (Class|FortGameStateAthena|GetAthenaPlaylistContextTags _returnvalue))
          (return _array_element))))))

(fn GetValidAbilitySystemForOwner
  (bind _returnvalue (CallFunction|GetAvatarasFortPlayerPawn))
  (bind _returnvalue_2 (Ability|GetFortAbilitySystemComponent _returnvalue))
  ;[0 0 0] Helper function for getting a the owning player's ability system if it exists
  (Utilities|IsValid _returnvalue
    (:Is%20Valid
      (Utilities|IsValid _returnvalue_2
        (:Is%20Valid
          (return _returnvalue_2))
        (:Is%20Not%20Valid
          _)))
    (:Is%20Not%20Valid
      _))
  (return 0))
