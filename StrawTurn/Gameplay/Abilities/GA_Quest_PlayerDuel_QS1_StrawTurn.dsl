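; Gameplay ability for the StrawTurn player-duel quest (QS1), a child of a parent duel bounty ability.
; On activation it copies its own StrawTurn success, poached and failed texts into the duel bounty text variables, then calls the parent's activation.
; Passive activation failure, ability end and poached-or-failed handling only forward to the parent, so the duel logic lives there.
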(event Ability|EventActivateAbility
  (Variables|DuelBountiesVariables|SetSuccessText (Variables|Default|GetStrawTurnSuccessText))
  (Variables|DuelBountiesVariables|SetPoachedText (Variables|Default|GetStrawTurnPoachedText))
  (Variables|DuelBountiesVariables|SetFailedText (Variables|Default|GetStrawTurnFailedText))
  (|CallParentFunction))

(event Ability|EventFailedToActivatePassiveAbility
  (params (ActorInfo :type Gameplay%20Ability%20Actor%20Info%20Structure))
  (|CallParentFunction ActorInfo))

(event Ability|EventOnEndAbility
  (params (bWasCancelled :type Boolean))
  (|CallParentFunction bWasCancelled))

(fn HandleDuelPoachedOrFailed
  (|CallParentFunction))
