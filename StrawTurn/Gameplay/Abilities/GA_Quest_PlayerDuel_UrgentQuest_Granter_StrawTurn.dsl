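; Urgent quest granter ability for StrawTurn player duels, acting on quest outcomes only when its owner carries the instigator tag.
; Success applies an effect and cue, then spawns squad-size-tiered loot with random mod sets on moddable weapons.
; Expiry applies a damage effect after a short race-avoiding delay.
; Failed or ended quests apply a poached/ended effect to living owners, or strip the cosmetic effect from dead ones.
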(event Ability|EventOnEndAbility
  (params (bWasCancelled :type Boolean))
  (bind _owningplayerpawn (Variables|Default|GetOwningPlayerPawn))
  (bind _returnvalue (Ability|GetFortAbilitySystemComponent _owningplayerpawn))
  (|CallParentFunction bWasCancelled)
  (Utilities|IsValid _returnvalue
    (:Is%20Valid
      (GameplayEffects|RemoveActiveEffectswithAppliedTags _returnvalue (Variables|Default|GetRebootedGE_Tag)))))

(event Ability|EventActivateAbility
  (|CallParentFunction))

(event EventOnUrgentQuestFailed
  (params (PlayerStateAthena :type Fort%20Player%20State%20Athena%20Object%20Reference) (UrgentQuestData :type Urgent%20Quest%20Data%20Structure%20%28by%20ref%29))
  ;[0 0 0] On urgent quest ended for any reason - remove cosmetic effect if the owner is also the duel instigator
  (|CallParentFunction PlayerStateAthena UrgentQuestData)
  (bind _bownerisinstigator (CallFunction|AbilityOwnerIsDuelInstigator))
  (if (and _bownerisinstigator (== (Variables|UrgentQuestVariables|GetUrgent%20Quest%20Tag) (Utilities|Struct|BreakUrgentQuestData UrgentQuestData)))
    (bind _abilitysystem (CallFunction|GetValidAbilitySystemforOwner))
    (if (not (FortPawn|IsDead (CallFunction|GetAvatarasFortPlayerPawn)))
      (Utilities|FlowControl|Delay 0.5)
      (GameplayEffects|ApplyGameplayEffectToSelf :self _abilitysystem :GameplayEffectClass (Variables|Default|GetDuelPoachedOrEndedEffect) :Level 0.0)
      (else
        (GameplayEffects|RemoveActiveEffectswithAppliedTags _abilitysystem "(GameplayTags=((TagName=\"GameplayEffect.StrawTurn.CosmeticEffect\")))")))))

(event EventOnUrgentQuestEnded
  (params (PlayerStateAthena :type Fort%20Player%20State%20Athena%20Object%20Reference) (UrgentQuestTag :type Gameplay%20Tag%20Structure%20%28by%20ref%29) (QuestState :type EFortQuestState%20Enum :default "Inactive") (CompletionPathContext :type Gameplay%20Tag%20Container%20Structure%20%28by%20ref%29))
  ;[0 0 0] On urgent quest ended for any reason - remove cosmetic effect if the owner is also the duel instigator
  (|CallParentFunction PlayerStateAthena UrgentQuestTag QuestState CompletionPathContext)
  (bind _bownerisinstigator (CallFunction|AbilityOwnerIsDuelInstigator))
  (if (and _bownerisinstigator (== (Variables|UrgentQuestVariables|GetUrgent%20Quest%20Tag) UrgentQuestTag))
    (bind _abilitysystem (CallFunction|GetValidAbilitySystemforOwner))
    (if (not (FortPawn|IsDead (CallFunction|GetAvatarasFortPlayerPawn)))
      (Utilities|FlowControl|Delay 0.5)
      (GameplayEffects|ApplyGameplayEffectToSelf :self _abilitysystem :GameplayEffectClass (Variables|Default|GetDuelPoachedOrEndedEffect) :Level 0.0)
      (else
        (GameplayEffects|RemoveActiveEffectswithAppliedTags _abilitysystem "(GameplayTags=((TagName=\"GameplayEffect.StrawTurn.CosmeticEffect\")))")))))

(event EventOnUrgentQuestExpired
  (params (UrgentQuestData :type Urgent%20Quest%20Data%20Structure%20%28by%20ref%29))
  ;[0 0 0] On time expired, call the damage GE
  (|CallParentFunction UrgentQuestData)
  (bind _bownerisinstigator (CallFunction|AbilityOwnerIsDuelInstigator))
  (if (and _bownerisinstigator (== (Variables|UrgentQuestVariables|GetUrgent%20Quest%20Tag) (Utilities|Struct|BreakUrgentQuestData UrgentQuestData)))
    (bind _abilitysystem (CallFunction|GetValidAbilitySystemforOwner))
    ;; Delay the damage slightly so that we don't hit a race condition with the urgent quest ending at the same time the participant is eliminated by the effect
    (Utilities|FlowControl|Delay 0.3)
    (GameplayEffects|ApplyGameplayEffectToSelf :self _abilitysystem :GameplayEffectClass (Variables|Default|GetDamageEffectClass) :Level 0.0)
    (Ability|EndAbility)))

(event EventAdditionalUrgentQuestEndedLogic
  (params (bTimedOutOrFailed :type Boolean))
  (bind _returnvalue (CallFunction|GetAvatarasFortPlayerPawn))
  ;[0 0 0] If quest was completed successfully, play the fanfare and give reward
  (|CallParentFunction bTimedOutOrFailed)
  (if (not bTimedOutOrFailed)
    (bind _bownerisinstigator (CallFunction|AbilityOwnerIsDuelInstigator))
    (if _bownerisinstigator
      (bind _abilitysystem (CallFunction|GetValidAbilitySystemforOwner))
      (GameplayEffects|ApplyGameplayEffectToSelf :self _abilitysystem :GameplayEffectClass (Variables|Default|GetSuccessEffectClass) :Level 0.0)
      (Ability|ExecuteGameplayCueOnOwner (Variables|Default|GetSuccessGameplayCueTag))
      (Utilities|IsValid _returnvalue
        (:Is%20Valid
          (bind (_lootgroupname _bfoundtier) (CallFunction|GetLootTierforCurrentPlaylist))
          (if _bfoundtier
            (bind _itempickupsspawned (Class|AthenaFunctionLibrary|SpawnLootfromTable _lootgroupname 1 -1 (Transformation|GetActorLocation _returnvalue) "0, 0, 0" 0 true true))
            (CallFunction|ApplyModstoWeaponsSpawned _itempickupsspawned)
            (Ability|EndAbility)))))))

(fn AbilityOwnerIsDuelInstigator
  (bind _returnvalue (CallFunction|GetAvatarasFortPlayerPawn))
  ;[0 0 0] Helper function for checking if the ability owner is also the duel instigator
  (Utilities|IsValid _returnvalue
    (:Is%20Valid
      (return (GameplayTags|HasMatchingGameplayTag _returnvalue "(TagName=\"Granted.Athena.JobBoardExperimental.Instigator\")")))))

(fn ApplyModsToWeaponsSpawned
  (params (LootSpawned :type Array%20of%20Fort%20Pickup%20Object%20References))
  ;[0 0 0] For each weapon item spawned, apply a modset to it (if that is supported)
  (for _array_element LootSpawned
    (bind _asfort_weapon_item_definition (Utilities|Casting|CastToFortWeaponItemDefinition (Pickup|GetPrimaryItemDefinition _array_element)))
    (if (Variables|/Script/FortniteGame.FortWeaponItemDefinition|GetbIsModifiableWeapon _asfort_weapon_item_definition)
      (bind _returnvalue (Class|FortWeaponModFunctionLibrary|GetRandomWeaponModSetData _asfort_weapon_item_definition))
      (Utilities|IsValid _returnvalue
        (:Is%20Valid
          (Class|FortWeaponModFunctionLibrary|ApplyModSettoPickup _returnvalue _array_element)))))
  (return))

(fn GetLootTierForCurrentPlaylist
  (bind (_value _returnvalue) (Utilities|Map|Find (Variables|Default|GetMapPlaylistToLootTier) (CallFunction|GetPlaylistSquadSizeTag)))
  ;[0 0 0] Maps the loot tier based on the playlist squad size and returns the name of the loot tier group
  (return _value _returnvalue))

(fn GetPlaylistSquadSizeTag
  (bind _returnvalue (Fortnite|Athena|GameState|GetGameStateAthena))
  ;[0 0 0] Get the playlist tag to determine if it's solo/squad etc.
  (Utilities|IsValid _returnvalue
    (:Is%20Valid
      (for _array_element (LocalVariables|GetPossiblePlaylistTags)
        (if (GameplayTags|TagContainer|CompareTagContainertoOtherTags _array_element (Class|FortGameStateAthena|GetAthenaPlaylistContextTags _returnvalue))
          (return _array_element))))))

(fn GetValidAbilitySystemForOwner
  (bind _returnvalue (CallFunction|GetAvatarasFortPlayerPawn))
  (bind _returnvalue_2 (Ability|GetFortAbilitySystemComponent _returnvalue))
  ;[0 0 0] Helper function for getting a the owning player's ability system if it exists
  (Utilities|IsValid _returnvalue
    (:Is%20Valid
      (Utilities|IsValid _returnvalue_2
        (:Is%20Valid
          (return _returnvalue_2))
        (:Is%20Not%20Valid
          _)))
    (:Is%20Not%20Valid
      _))
  (return 0))
