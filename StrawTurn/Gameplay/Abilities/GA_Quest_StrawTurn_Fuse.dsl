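; Straw Turn quest ability that lights a timed fuse effect on the player's hat once a target-found event arrives.
; After a short delay it applies a replicated fuse-visuals gameplay effect, then fetches the Hat part modifier on cosmetic-capable machines to set its particle parameters from the total duel time.
; A separate sound event, not called within this graph, schedules a crossfade from ambient to warning audio near the fuse's end.
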(event Ability|EventFailedToActivatePassiveAbility
  (params (ActorInfo :type Gameplay%20Ability%20Actor%20Info%20Structure))
  (|CallParentFunction ActorInfo))

(event Ability|EventOnEndAbility
  (params (bWasCancelled :type Boolean))
  (|CallParentFunction bWasCancelled))

(event Ability|EventActivateAbility
  ;[0 0 0] Activate ability (server) - wait for the target to be found and the part modifier to exist
  (|CallParentFunction)
  (Ability|Async|WaitGameplayEventToActor (CallFunction|GetAvatarasFortPlayerPawn) (Variables|Default|GetTargetFoundTag) true
    (:EventReceived
      (bind _returnvalue (Utilities|Time|SetTimerbyEvent (delegate Custom|PostDelay) 0.5)) ; Give a small amount of time before igniting the fuse
      (Variables|Default|SetDelayTimerHandle _returnvalue))))

(event Custom|PostDelay
  ;[0 0 0] Activate ability (server) - wait for the target to be found and the part modifier to exist
  (GameplayEffects|ApplyGameplayEffectToSelf :self (Ability|GetFortAbilitySystemComponent (CallFunction|GetAvatarasFortPlayerPawn)) :GameplayEffectClass (Variables|Default|GetGE_ReplicatedFuseVisuals) :Level 0.0)
  (CallFunction|GetReferencetoPartModifier))

(event Custom|ManageSoundForOwningClient
  (bind _totaldueltime (Variables|Default|GetTotalDuelTime))
  (bind _returnvalue (ScalableFloat|GetValueatLevel _totaldueltime))
  (bind _fusesoundwarningcue (Variables|Audio|GetFuseSoundWarningCue))
  (bind _duration (Variables|/Script/Engine.SoundBase|GetDuration _fusesoundwarningcue))
  (bind _returnvalue_2 (- _returnvalue _duration))
  ;[0 0 0] Manage Sound for Owning Client (server) - starts timer to determine when warning sound should play
  (if (> _returnvalue_2 0)
    (bind _output_get (Variables|Default|SetTimeElapsedBeforeWarning _returnvalue_2))
    (bind _returnvalue_3 (Utilities|Time|SetTimerbyEvent (EventDispatchers|CreateEvent :!event Custom|SwapSoundToWarning) _output_get false true)) ; Swap sound to warning loop on fuse when the time elapses
    (Variables|Audio|SetSoundTimerHandle _returnvalue_3)))

(event Custom|GetReferenceToPartModifier
  (bind _returnvalue (CallFunction|GetAvatarasFortPlayerPawn))
  (bind _returnvalue_2 (FortPawn|GetModifierforPartType _returnvalue "Hat"))
  ;[0 0 0] Client only - get the hat modifier for part type if it exists
  (Networking|CanExecuteCosmeticEvents
    (:True
      (Utilities|IsValid _returnvalue_2
        (:Is%20Valid
          (Variables|Default|SetPartModifier _returnvalue_2)
          (CallFunction|UpdateFuse))))))

(event Custom|UpdateFuse
  (bind _returnvalue (Class|CustomCharacterPartModifier|GetIdleVFXComponent (Variables|Default|GetPartModifier)))
  (bind _partmodifier (Variables|Default|GetPartModifier))
  ;[0 0 0] Client only - Update the fuse visuals and sounds
  (Networking|CanExecuteCosmeticEvents
    (:True
      ;; Set visual parameters on the FX fuse
      (Utilities|IsValid _partmodifier
        (:Is%20Valid
          (Effects|Components|ParticleSystem|SetBoolParameter _returnvalue "PlayFuse" true)
          (Effects|Components|ParticleSystem|SetFloatParameter _returnvalue "FuseTime" (ScalableFloat|GetValueatLevel (Variables|Default|GetTotalDuelTime))))))))

(event Custom|SwapSoundToWarning
  (bind _ambientsoundreference (Variables|Audio|GetAmbientSoundReference))
  ;[0 0 0] Client only - Update the fuse visuals and sounds
  ;; Play the fuse warning sound when the time runs down
  (Networking|CanExecuteCosmeticEvents
    (:True
      (Utilities|IsValid _ambientsoundreference
        (:Is%20Valid
          (Audio|Components|Audio|FadeOut _ambientsoundreference 3.0 0.0 "Linear" :!class /Script/Engine.AudioComponent)
          (bind _returnvalue (Audio|SpawnSoundAttached :Sound (Variables|Audio|GetFuseSoundWarningCue) :AttachToComponent (Class|CustomCharacterPartModifier|GetIdleVFXComponent (Variables|Default|GetPartModifier)) :AttachPointName "FX_Fuse" :LocationType "KeepRelativeOffset" :bStopWhenAttachedToDestroyed true :VolumeMultiplier 1.0 :PitchMultiplier 1.0 :StartTime 0.0 :bAutoDestroy true))
          (Variables|Audio|SetWarningSoundReference _returnvalue))))))
