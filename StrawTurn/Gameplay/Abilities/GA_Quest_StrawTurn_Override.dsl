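; StrawTurn quest gameplay ability that applies a head cosmetic to the avatar pawn while active, by granting a gadget item.
; On activation it runs the parent logic, waits a configurable delay and for playlist data, then caches the pawn and calls SwapHead.
; SwapHead gives the item to the player controller, or to an Athena AI bot controller otherwise, then fires a gameplay cue.
; On end, RestoreHead removes the item from that controller.
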(event Ability|EventActivateAbility
  (bind _returnvalue (CallFunction|GetAvatarasFortPlayerPawn))
  ;[0 0 0] Apply the head cosmetic after a short delay
  (|CallParentFunction)
  (Utilities|FlowControl|Delay (Variables|Default|GetDelayBeforeApplying))
  (Class|FortAsyncActionCurrentPlaylistReady|PlaylistDataReadyAsync
    (:OnReady
      (Utilities|IsValid _returnvalue
        (:Is%20Valid
          (Variables|Default|SetInstigatingPlayerPawn _returnvalue)
          (CallFunction|SwapHead))))))

(event Ability|EventOnEndAbility
  (params (bWasCancelled :type Boolean))
  ;[0 0 0] Remove cosmetic when ability ends
  (|CallParentFunction bWasCancelled)
  (CallFunction|RestoreHead))

(fn RestoreHead
  (bind _instigatingplayerpawn (Variables|Default|GetInstigatingPlayerPawn))
  (bind _returnvalue (Fortnite|GetFortPlayerControllerfromActor _instigatingplayerpawn))
  ;[0 0 0] Check if this is a player or phoebe, and remove the item accordingly
  (Utilities|IsValid _returnvalue
    (:Is%20Valid
      (Inventory|RemoveItemFromPlayer :PlayerController _returnvalue :ItemDefinition (Variables|Default|GetPlayerGadgetItem) :AmountToRemove -1 :bForceRemoval true :bSkipVerbRemoval false))
    (:Is%20Not%20Valid
      (bind _asfort_athena_aibot_controller (Utilities|Casting|CastToFortAthenaAIBotController (AI|GetAIController _instigatingplayerpawn))
        (:then
          (Inventory|RemoveItemfromInventoryOwner _asfort_athena_aibot_controller (Variables|Default|GetPlayerGadgetItem)))
        (:CastFailed
          (return)))))
  (return))

(fn SwapHead
  (bind _instigatingplayerpawn (Variables|Default|GetInstigatingPlayerPawn))
  (bind _returnvalue (Fortnite|GetFortPlayerControllerfromActor _instigatingplayerpawn))
  (Utilities|IsValid (Variables|Default|GetPlayerGadgetItem)
    (:Is%20Valid
      (Utilities|IsValid _returnvalue
        (:Is%20Valid
          (bind _returnvalue_2 (Inventory|GiveItemtoInventoryOwner _returnvalue (Variables|Default|GetPlayerGadgetItem)))
          (LocalVariables|SetInventoryItem _returnvalue_2))
        (:Is%20Not%20Valid
          (bind _asfort_athena_aibot_controller (Utilities|Casting|CastToFortAthenaAIBotController (AI|GetAIController _instigatingplayerpawn))
            (:then
              (bind _returnvalue_3 (Inventory|GiveItemtoInventoryOwner :InventoryOwner _asfort_athena_aibot_controller :ItemDefinition (Variables|Default|GetPlayerGadgetItem) :NumberToGive 1 :bNotifyPlayer false :ItemLevel -1 :PickupInstigatorHandle -1 :bUseItemPickupAnalyticEvent false :WeaponAmmoOverride -1))
              (LocalVariables|SetInventoryItem _returnvalue_3))
            (:CastFailed
              (return)))))
      (Ability|ExecuteGameplayCueOnOwner (Variables|Default|GetGameplay%20Cue%20Tag))
      (return))
    (:Is%20Not%20Valid
      (return))))
